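# Looping hover sound cue for the Data Drone in a Season 20 Battle Pass Week 4 narrative quest, in the ambient sound class.
# It reuses a supply drop drone loop, fed through an intensified, smoothed Doppler node into an attenuation node and a single-input mixer that trims volume.
# The attenuation node overrides its settings inline with logarithmic distance falloff and distance-based low-pass filtering.
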SoundCue /BattlePassS20/QuestActors/Narrative/Week_04/SFX/S20_DataDrone_Hover_Loop_Cue
root=Mixer_1
nodes [
 {
  id=Attenuation_1
  children [Doppler_0]
  properties {
   bOverrideAttenuation=true
   attenuationOverrides {
    bAttenuateWithLPF=true
    bEnableLogFrequencyScaling=true
    distanceAlgorithm=Logarithmic
    falloffDistance=1000
    hPFFrequencyAtMin=1000
    lPFFrequencyAtMin=1076.354
    lPFRadiusMax=2583.3145
    lPFRadiusMin=1517.5007
    stereoSpread=100
    attenuationShapeExtents {x=100}
   }
  }
 }
 {id=Doppler_0, children [WavePlayer_5], properties {bUseSmoothing=true, dopplerIntensity=3}}
 {id=Mixer_1, children [Attenuation_1], properties {inputVolume [0.85]}}
 {id=WavePlayer_5, properties {bLooping=true, soundWaveAssetPtr=/Game/Athena/Sounds/SupplyDrops/Drone/Athena_SupplyDrop_Drone_Loop_04}}
]
properties {
 maxDistance=1100
 soundClassObject=&Fort_Ambient_Athena
 volumeMultiplier=1.247
 attenuationOverrides {bEnableListenerFocus=true, nonFocusVolumeAttenuation=0.5}
 concurrencyOverrides {maxCount=4, resolutionRule=StopQuietest}
}
