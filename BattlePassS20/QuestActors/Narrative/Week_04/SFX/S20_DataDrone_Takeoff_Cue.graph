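# Sound cue for the takeoff of the Season 20 data drone, a Week 4 narrative quest actor.
# A Random node picks one of three equally weighted takeoff waves and feeds a Modulator that adds slight pitch variation.
# It plays under the quests sound class, with overridden attenuation alongside a reference to the Season 20 quest VO attenuation asset.
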SoundCue /BattlePassS20/QuestActors/Narrative/Week_04/SFX/S20_DataDrone_Takeoff_Cue
root=Modulator_0
nodes [
 {id=Modulator_0, children [Random_0], properties {pitchMax=1.02, pitchMin=0.98, volumeMax=1}}
 {id=Random_0, children [WavePlayer_0,WavePlayer_1,WavePlayer_2], properties {weights [1,1,1]}}
 {id=WavePlayer_0, properties {soundWaveAssetPtr=&/S20_DataDrone_Takeoff_01}}
 {id=WavePlayer_1, properties {soundWaveAssetPtr=&/S20_DataDrone_Takeoff_02}}
 {id=WavePlayer_2, properties {soundWaveAssetPtr=&/S20_DataDrone_Takeoff_03}}
]
properties {
 attenuationSettings=/BattlePassS20/Audio/VO/S20_Quest_Att
 bOverrideAttenuation=true
 maxDistance=1300
 soundClassObject=&Fort_Quests
 attenuationOverrides {falloffDistance=1000, attenuationShapeExtents {x=300}}
}
