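# Sound cue for a tank deflating in a Battle Pass Season 20 Week 4 narrative quest, routed through the quest sound class.
# A random node picks one of five equally weighted deflate wave variations and feeds a modulator that varies pitch while holding volume fixed.
# Its own attenuation override sets logarithmic falloff with distance low-pass filtering, stereo spread and listener focus, which treats off-focus sounds differently.
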SoundCue /BattlePassS20/QuestActors/Narrative/Week_04/SFX/S20_W04_Q02_TankDeflate_Cue
root=Modulator_0
nodes [
 {id=Modulator_0, children [Random_0], properties {volumeMax=1, volumeMin=1}}
 {id=Random_0, children [WavePlayer_0,WavePlayer_1,WavePlayer_2,WavePlayer_3,WavePlayer_4], properties {weights [1,1,1,1,1]}}
 {id=WavePlayer_0, properties {soundWaveAssetPtr=&/S20_W04_Q02_TankDeflate_05}}
 {id=WavePlayer_1, properties {soundWaveAssetPtr=&/S20_W04_Q02_TankDeflate_01}}
 {id=WavePlayer_2, properties {soundWaveAssetPtr=&/S20_W04_Q02_TankDeflate_02}}
 {id=WavePlayer_3, properties {soundWaveAssetPtr=&/S20_W04_Q02_TankDeflate_03}}
 {id=WavePlayer_4, properties {soundWaveAssetPtr=&/S20_W04_Q02_TankDeflate_04}}
]
properties {
 bOverrideAttenuation=true
 maxDistance=9250
 soundClassObject=&Fort_Quests
 volumeMultiplier=0.9
 attenuationOverrides {
  bAttenuateWithLPF=true
  bEnableListenerFocus=true
  distanceAlgorithm=Logarithmic
  falloffDistance=7750
  lPFFrequencyAtMax=12000
  nonFocusAzimuth=45
  nonFocusDistanceScale=0.5
  nonFocusVolumeAttenuation=0.2
  stereoSpread=1000
  attenuationShapeExtents {x=1500}
 }
}
